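; PCG graph that generates industrial cable runs along input actor splines, with optional auxiliary cables and scattered detail meshes.
; The optionally resampled spline is either noise-displaced or looped into several disturbed cables, chosen by cable count, using duplicated points as an iteration workaround.
; These are sectioned, randomly scaled and rebuilt as spline meshes.
; Rebuilt splines also seed a randomly chosen subset of detail spawns and a second loop for rescaled auxiliary cables.
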(graph Cable_Generator
  (params (Input_Spline :type (Array (SoftObject /Script/Engine.Actor))) (detail_Meshes :type (Array (SoftObject /Script/Engine.StaticMesh)) :default [/PCG/DebugObjects/PCG_Cube.PCG_Cube]) (Spline_sample :type Int :default 14) (Number_of_Cables :type Int :default 5) (Max_Width :type Double :default 1.0) (Min_Wdth :type Double :default 1.0) (Section_every_x_cm :type Double :default 1e+03) (Distrubance_amount :type Double :default 3e+02) (Resample_spline :type Bool :default true) (Seed :type Int) (Aux_Cables :type Bool :default true) (Number_of_Aux_Cables :type Int :default 2) (Aux_Cables_Disturbance :type Double :default 5e+02) (Aux_Cables_Scale_mult :type Double :default 0.5) (Details_Density :type Int :default 1) (Details_Amount :type Double :default 0.3))
  (bind AttributeCompareOp_0 (Attribute%20Compare%20Op :InA Number_of_Cables))
  (bind GetSplineData_0 (Get%20Spline%20Data :ActorSelector (struct :ActorFilter FromInput) :In Input_Spline))
  (bind Branch_2 (Branch :bOutputToB-param Resample_spline))
  (bind Subgraph_21 (Subgraph :Graph /PCGPrimitives/Primitives/Transform/Transform_Resample.Transform_Resample :Subdivision Spline_sample))
  (bind Branch_1 (Branch))
  (bind FilterDataByType_3 (Filter%20Data%20By%20Type :TargetType (struct :Ids [(struct :Struct /Script/PCG.PCGDataTypeInfoSpline)])))
  (bind Loop_0 (Loop :Graph /PCGPrimitives/Examples/Industrial/Cable_Generator_Loop.Cable_Generator_Loop :disturbance_amount Distrubance_amount :seed-param Seed))
  (bind AttributeMathsOp_1 (Attribute%20Maths%20Op :Operation Multiply :InA Distrubance_amount))
  (bind AttributeNoise_2 (Attribute%20Noise :InputSource $Position :Mode Add :NoiseMax-param Distrubance_amount :Seed-param Seed))
  (bind Subgraph_22 (Subgraph :Graph /PCGPrimitives/Primitives/Place/Place_OnSpline.Place_OnSpline :Spacing Section_every_x_cm))
  (bind ExtractAttribute_0 (Extract%20Attribute%20at%20Index :bForceOutputAttributeToBeInElementsDomain true :InputSource $Scale.X :OutputAttributeName scale))
  (bind MergeAttributes_0 (Merge%20Attributes))
  (bind AttributeNoise_1 (Attribute%20Noise :InputSource @Last :NoiseMax 4.0 :NoiseMin 0.1 :NoiseMax-param Max_Width :NoiseMin-param Min_Wdth))
  (bind AttributeCast_0 (Attribute%20Cast :OutputType Vector))
  (bind AttributeMathsOp_2 (Attribute%20Maths%20Op :InputSource1 scale :Operation Multiply))
  (bind AttributePartition_2 (Attribute%20Partition :PartitionAttributeSelectors [$Index]))
  (bind CopyAttributes_1 (Copy%20Attributes :InputSource @Last :OutputTarget $Scale))
  (bind Branch_0 (Branch))
  (bind CreateConstant_2 (Create%20Constant :AttributeTypes (struct :Type Integer32 :Int32Value 2)))
  (bind FilterDataByType_8 (Filter%20Data%20By%20Type :TargetType (struct :Ids [(struct :Struct /Script/PCG.PCGDataTypeInfoPoint)])))
  (bind CreateSpline_3 (Create%20Spline :InterpTypeAttribute None))
  (bind SpawnSplineMesh_1 (Spawn%20Spline%20Mesh :SplineMeshDescriptor (struct :StaticMesh /PCGPrimitives/Meshes/SM_Cable_Module.SM_Cable_Module)))
  (bind CopyAttributes_0 (Copy%20Attributes :InputSource @Last :OutputTarget $Scale))
  (bind Subgraph_3 (Subgraph :Graph /PCGPrimitives/Primitives/Transform/Transform_Resample.Transform_Resample :Subdivision Details_Density))
  (bind FilterDataByType_5 (Filter%20Data%20By%20Type :TargetType (struct :Ids [(struct :Struct /Script/PCG.PCGDataTypeInfoPolyline)])))
  (bind GetSplineControlPoints_0 (Get%20Spline%20Control%20Points))
  (bind MergePoints_0 (Merge%20Points))
  (bind CopyAttributes_3 (Copy%20Attributes :InputSource $Index :OutputTarget $Seed))
  (bind RandomChoice_0 (Random%20Choice :bFixedMode false :Ratio-param Details_Amount))
  (bind Subgraph_2 (Subgraph :Graph /PCGPrimitives/Primitives/Spawn/Spawn_Assets.Spawn_Assets :Assets detail_Meshes))
  (bind Branch_3 (Branch :bOutputToB-param Aux_Cables))
  (bind AttributeMathsOp_5 (Attribute%20Maths%20Op :InA Seed))
  (bind Loop_2 (Loop :Graph /PCGPrimitives/Examples/Industrial/Cable_Generator_Loop_Auxiliary_cables.Cable_Generator_Loop_Auxiliary_cables :disturbance_amount Aux_Cables_Disturbance))
  (bind AttributeMathsOp_6 (Attribute%20Maths%20Op :InputSource1 $Scale :Operation Multiply :InB Aux_Cables_Scale_mult))
  (bind CreateConstant_3 (Create%20Constant :AttributeTypes (struct :Type Integer32 :Int32Value 2)))
  (bind SpawnSplineMesh_2 (Spawn%20Spline%20Mesh :SplineMeshDescriptor (struct :StaticMesh /PCGPrimitives/Meshes/SM_Cable_Module.SM_Cable_Module)))
  ;[0.15 0.15 0.15 0.5] This is a work around to creat iteration. we just neex X number of spline...
  (bind AttributeMathsOp_0 (Attribute%20Maths%20Op :Operation Subtract :InA Number_of_Cables))
  (bind CreatePoints_0 (Create%20Points))
  (bind DuplicatePoint_0 (Duplicate%20Point))
  (bind AttributePartition_1 (Attribute%20Partition :PartitionAttributeSelectors [$Index]))
  ;[0.15 0.15 0.15 0.5] This is a work around to create iteration. we just neex X number of spline...
  (bind AttributeMathsOp_3 (Attribute%20Maths%20Op :Operation Subtract :InA Number_of_Aux_Cables))
  (bind CreatePoints_1 (Create%20Points))
  (bind DuplicatePoint_1 (Duplicate%20Point))
  (bind AttributePartition_3 (Attribute%20Partition :PartitionAttributeSelectors [$Index]))
  (connect AttributeNoise_1 Out AttributeCast_0 In)
  (connect AttributeCast_0 Out AttributeMathsOp_2 InA)
  (connect Loop_2 Out AttributeMathsOp_6 InA)
  (connect MergeAttributes_0 Out AttributeNoise_1 In)
  (connect Branch_1 Output%20B AttributeNoise_2 In)
  (connect AttributeMathsOp_1 Out AttributeNoise_2 NoiseMin)
  (connect DuplicatePoint_0 Out AttributePartition_1 In)
  (connect AttributeMathsOp_2 Out AttributePartition_2 In)
  (connect DuplicatePoint_1 Out AttributePartition_3 In)
  (connect AttributeCompareOp_0 Out Branch_1 bOutputToB)
  (connect Branch_2 Output%20A Branch_1 In)
  (connect Subgraph_21 Out Branch_1 In)
  (connect GetSplineData_0 Out Branch_2 In)
  (connect CreateSpline_3 Out Branch_3 In)
  (connect AttributePartition_2 Out CopyAttributes_1 Source)
  (connect Subgraph_22 Out CopyAttributes_1 Target)
  (connect MergePoints_0 Out CopyAttributes_3 Source)
  (connect MergePoints_0 Out CopyAttributes_3 Target)
  (connect FilterDataByType_8 InsideFilter CreateSpline_3 In)
  (connect CreatePoints_0 Out DuplicatePoint_0 In)
  (connect AttributeMathsOp_0 Out DuplicatePoint_0 Iterations)
  (connect CreatePoints_1 Out DuplicatePoint_1 In)
  (connect AttributeMathsOp_3 Out DuplicatePoint_1 Iterations)
  (connect Subgraph_22 Out ExtractAttribute_0 In)
  (connect Branch_1 Output%20A FilterDataByType_3 In)
  (connect Subgraph_3 Out FilterDataByType_5 In)
  (connect CopyAttributes_1 Out FilterDataByType_8 In)
  (connect FilterDataByType_5 InsideFilter GetSplineControlPoints_0 In)
  (connect AttributePartition_1 Out Loop_0 In)
  (connect FilterDataByType_3 InsideFilter Loop_0 spline_in)
  (connect AttributePartition_3 Out Loop_2 In)
  (connect AttributeMathsOp_5 Out Loop_2 seed)
  (connect Branch_3 Output%20B Loop_2 spline_in)
  (connect ExtractAttribute_0 Out MergeAttributes_0 In)
  (connect GetSplineControlPoints_0 Out MergePoints_0 In)
  (connect CopyAttributes_3 Out RandomChoice_0 In)
  (connect CreateConstant_2 Out SpawnSplineMesh_1 ForwardAxis)
  (connect CreateSpline_3 Out SpawnSplineMesh_1 In)
  (connect CreateConstant_3 Out SpawnSplineMesh_2 ForwardAxis)
  (connect AttributeMathsOp_6 Out SpawnSplineMesh_2 In)
  (connect RandomChoice_0 Chosen Subgraph_2 PrimaryInput)
  (connect CreateSpline_3 Out Subgraph_3 PrimaryInput)
  (connect Branch_2 Output%20B Subgraph_21 PrimaryInput)
  (connect AttributeNoise_2 Out Subgraph_22 PrimaryInput)
  (connect Loop_0 Out Subgraph_22 PrimaryInput))
