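; PCG loop body for an industrial cable generator example that jitters the positions of an incoming spline to vary each cable.
; Attribute Noise adds random offsets to position, with the disturbance amount as maximum and a multiplied copy of it as minimum.
; The noise seed comes from the seed parameter, passed through a math op, squared, then combined with the loop index, so each iteration gets different noise.
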(graph Cable_Generator_Loop
  (params (disturbance_amount :type Double) (seed :type Int))
  (inputs (spline_in :type Spline))
  (bind AttributeMathsOp_1 (Attribute%20Maths%20Op :Operation Multiply :InA disturbance_amount))
  (bind AttributeMathsOp_2 (Attribute%20Maths%20Op :InA seed))
  (bind AttributeMathsOp_3 (Attribute%20Maths%20Op :Operation Multiply))
  (bind GetLoopIndex_0 (Get%20Loop%20Index))
  (bind AttributeMathsOp_4 (Attribute%20Maths%20Op))
  (bind AttributeNoise_0 (Attribute%20Noise :InputSource $Position :Mode Add :NoiseMax-param disturbance_amount))
  (connect AttributeMathsOp_2 Out AttributeMathsOp_3 InA)
  (connect AttributeMathsOp_2 Out AttributeMathsOp_3 InB)
  (connect AttributeMathsOp_3 Out AttributeMathsOp_4 InA)
  (connect GetLoopIndex_0 Out AttributeMathsOp_4 InB)
  (connect DefaultInputNode spline_in AttributeNoise_0 In)
  (connect AttributeMathsOp_1 Out AttributeNoise_0 NoiseMin)
  (connect AttributeMathsOp_4 Out AttributeNoise_0 Seed)
  (output Out AttributeNoise_0))
