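; PCG loop-body graph for the industrial cable generator example, handling auxiliary cables for an input spline.
; It extracts an index attribute from the spline and converts it to point data for a nested inner-loop graph, alongside the spline itself.
; The outer loop index is offset and multiplied by the seed parameter to vary each iteration's seed, while disturbance amount passes through to the inner graph, whose result is the output.
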(graph Cable_Generator_Loop_Auxiliary_cables
  (params (disturbance_amount :type Double) (seed :type Int))
  (inputs (spline_in :type Spline))
  (bind GetLoopIndex_0 (Get%20Loop%20Index))
  (bind AttributeMathsOp_1 (Attribute%20Maths%20Op))
  (bind AttributeMathsOp_0 (Attribute%20Maths%20Op :Operation Multiply :InA seed))
  (bind ExtractAttribute_0 (Extract%20Attribute%20at%20Index :bForceOutputAttributeToBeInElementsDomain true :InputSource $Index))
  (bind AttributeSetToPoint_0 (Attribute%20Set%20To%20Point))
  (bind Loop_0 (Loop :Graph /PCGPrimitives/Examples/Industrial/Cable_Generator_innerLoop_Auxiliary_cables.Cable_Generator_innerLoop_Auxiliary_cables :disturbance_amount disturbance_amount))
  (connect AttributeMathsOp_1 Out AttributeMathsOp_0 InB)
  (connect GetLoopIndex_0 Out AttributeMathsOp_1 InA)
  (connect ExtractAttribute_0 Out AttributeSetToPoint_0 In)
  (connect DefaultInputNode spline_in ExtractAttribute_0 In)
  (connect AttributeSetToPoint_0 Out Loop_0 In)
  (connect AttributeMathsOp_0 Out Loop_0 seed)
  (connect DefaultInputNode spline_in Loop_0 spline_in)
  (output Out Loop_0))
