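; PCG loop-body subgraph for an industrial cable generator that produces a randomly displaced auxiliary cable spline on each iteration.
; The loop index selects one entry from the input splines, and Attribute Noise adds an offset to its positions, bounded by values derived from disturbance_amount.
; The noise seed squares a value derived from the seed parameter and combines it with the loop index, so each cable varies differently.
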(graph Cable_Generator_innerLoop_Auxiliary_cables
  (params (disturbance_amount :type Double) (seed :type Int))
  (inputs (spline_in :type Spline))
  (bind AttributeMathsOp_1 (Attribute%20Maths%20Op :Operation Multiply :InA disturbance_amount))
  (bind AttributeMathsOp_2 (Attribute%20Maths%20Op :InA seed))
  (bind AttributeMathsOp_3 (Attribute%20Maths%20Op :Operation Multiply))
  (bind GetLoopIndex_0 (Get%20Loop%20Index))
  (bind AttributeMathsOp_4 (Attribute%20Maths%20Op))
  (bind GetLoopIndex_2 (Get%20Loop%20Index))
  (bind FilterDataByIndex_0 (Filter%20Data%20By%20Index))
  (bind AttributeNoise_0 (Attribute%20Noise :InputSource $Position :Mode Add :NoiseMax-param disturbance_amount))
  (connect AttributeMathsOp_2 Out AttributeMathsOp_3 InA)
  (connect AttributeMathsOp_2 Out AttributeMathsOp_3 InB)
  (connect AttributeMathsOp_3 Out AttributeMathsOp_4 InA)
  (connect GetLoopIndex_0 Out AttributeMathsOp_4 InB)
  (connect FilterDataByIndex_0 InsideFilter AttributeNoise_0 In)
  (connect AttributeMathsOp_1 Out AttributeNoise_0 NoiseMin)
  (connect AttributeMathsOp_4 Out AttributeNoise_0 Seed)
  (connect DefaultInputNode spline_in FilterDataByIndex_0 In)
  (connect GetLoopIndex_2 Out FilterDataByIndex_0 SelectedIndices)
  (output Out AttributeNoise_0))
